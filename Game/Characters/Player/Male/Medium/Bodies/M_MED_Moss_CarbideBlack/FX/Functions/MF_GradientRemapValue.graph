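# Material function in the TuataraVFX category that remaps a 0-1 gradient so a source value lands at a destination value.
# It scales the range below the source to zero-to-destination and the range above to destination-to-one, both linearly.
# Frac and Ceil mask the two halves without branching before they are summed, leaving zero exactly at the source point.
# Previews use texture-coordinate U as the gradient and DebugTimeSine as the source.
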MaterialFunction /Game/Characters/Player/Male/Medium/Bodies/M_MED_Moss_CarbideBlack/FX/Functions/MF_GradientRemapValue
graph {
 properties {description=Remaps a point in a 0-1 gradient, libraryCategoriesText [- TuataraVFX]}
 nodes [
  n1 = TextureCoordinate()
  n2 = Saturate(None=FunctionInput(Preview=ComponentMask(None=@n1) {r=true}) {bUsePreviewValueAsDefault=true, inputName=Gradient, inputType=FunctionInput_Scalar})
  n3 = FunctionInput(Preview=MaterialFunctionCall() {materialFunction=&DebugTimeSine}.Result) {
   bUsePreviewValueAsDefault=true
   desc=Value to remap
   description=Value to remap
   inputName=Value Source
   inputType=FunctionInput_Scalar
   sortPriority=1
   previewValue=<0.75 0 0 1>
  }
  n4 = Saturate(None=@n3)
  n5 = Frac(None=Saturate(None=Divide(A=@n2, B=@n4)))
  n6 = FunctionInput() {
   bUsePreviewValueAsDefault=true
   desc=Value to map source to
   description=Value to map source to
   inputName=Value Destination
   inputType=FunctionInput_Scalar
   sortPriority=2
   previewValue=<0.5 0 0 1>
  }
  n7 = Saturate(None=@n6)
  n8 = Subtract(A=@n2, B=@n4)
  n9 = Ceil(None=@n8)
  n10 = Subtract(B=@n4)
  n11 = Add(A=@n7, B=Multiply(A=Subtract(B=@n7), B=Divide(A=@n8, B=@n10)))
  FunctionOutput(None=Add(A=Multiply(A=@n5, B=@n7) {desc=First part of the gradient}, B=Multiply(A=@n9, B=@n11) {desc=Second part of the gradient})) {description=Gradient reflected around centerpoint}
 ]
}
