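# MF_RandomValue is a TuataraVFX material function, stored with the Moss CarbideBlack player body FX, that turns a seed into a random value.
# The Float3 seed defaults to world-space object position, so each object can get its own result.
# The seed is floored and offset by a half to sample Vector Noise at a fixed point per unit cell.
# Outputs provide the full RGB vector and separate R, G and B scalars.
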MaterialFunction /Game/Characters/Player/Male/Medium/Bodies/M_MED_Moss_CarbideBlack/FX/Functions/MF_RandomValue
graph {
 properties {description="Takes in a seed (usually object position) and returns a random value.", libraryCategoriesText [- TuataraVFX]}
 nodes [
  n1 = FunctionInput(Preview=ObjectPositionWS()) {bUsePreviewValueAsDefault=true, description="Random Float3 seed, defaults to Object Position.", inputName=Seed}
  n2 = VectorNoise(World Position=Add(A=Floor(None=@n1)) {constB=0.5})
  FunctionOutput(None=ComponentMask(None=@n2) {r=true}) {description=Random Scalar Value, outputName=R}
  FunctionOutput(None=@n2) {description=Random Vector3 Value, outputName=RGB, sortPriority=3}
  FunctionOutput(None=ComponentMask(None=@n2) {g=true}) {description=Random Scalar Value, outputName=G, sortPriority=1}
  FunctionOutput(None=ComponentMask(None=@n2) {b=true}) {description=Random Scalar Value, outputName=B, sortPriority=2}
 ]
}
