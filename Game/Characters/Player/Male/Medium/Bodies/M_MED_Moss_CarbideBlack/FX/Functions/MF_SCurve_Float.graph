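# Material function in the Tuatara VFX library, under a player body's FX folder, that applies an adjustable S-curve to a scalar 0-1 input.
# The lower half is remapped to 0-1 and raised to the Power input, while the upper half gets the inverted, mirrored power curve.
# Both halves are rescaled into their output ranges and joined by a Lerp, using Ceil of the upper curve as a branchless selector.
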MaterialFunction /Game/Characters/Player/Male/Medium/Bodies/M_MED_Moss_CarbideBlack/FX/Functions/MF_SCurve_Float
graph {
 properties {description=Applies a S Curve to a 0-1 input, libraryCategoriesText [- Tuatara VFX]}
 nodes [
  n1 = FunctionInput(Preview=ComponentMask(None=TextureCoordinate()) {r=true}) {bUsePreviewValueAsDefault=true, inputType=FunctionInput_Scalar}
  n2 = FunctionInput() {bUsePreviewValueAsDefault=true, inputName=Power, inputType=FunctionInput_Scalar, previewValue=<1 0 0 1>}
  n3 = Power(Base=Saturate(None=Multiply(A=@n1) {constB=2}), Exp=@n2)
  n4 = Saturate(None=ConstantBiasScale(None=@n1) {bias=-0.5, scale=2})
  n5 = OneMinus(None=Power(Base=OneMinus(None=@n4), Exp=@n2))
  FunctionOutput(None=LinearInterpolate(A=Divide(A=@n3), B=ConstantBiasScale(None=@n5), Alpha=Ceil(None=@n5)))
 ]
}
