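# Translucent unlit two-sided butterfly material for Niagara mesh particles in the Moss CarbideBlack character FX, animated in the shader.
# Per-particle randomized time drives UV-masked wing rotation about the particle pivot plus a delayed, randomized vertical flutter, with PreviousFrameSwitch supporting motion blur.
# Packed texture channels mask body, spots and opacity over randomly blended wing colors.
# Face side and flap phase shade the emissive, and camera depth fade modulates opacity.
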Material /Game/Characters/Player/Male/Medium/Bodies/M_MED_Moss_CarbideBlack/FX/M_Moss_Butterfly
graph {
 properties {
  bOutputTranslucentVelocity=true
  bUsedWithNiagaraMeshParticles=true
  blendMode=BLEND_Translucent
  pixelDepthOffsetMode=PDOM_Legacy
  shadingModel=MSM_Unlit
  twoSided=true
  usedShadingModels=Unlit
 }
 nodes [
  n1 = Normalize(VectorInput=Transform(None=Constant3Vector() {constant=<rgba -1 0 0 0>}) {transformSourceType=TRANSFORMSOURCE_ParticleWorld})
  n2 = Add(A=ScalarParameter() {defaultValue=2, parameterName=Flapping Speed}, B=ParticleRandom())
  n3 = Multiply(A=Add(A=Time(), B=ParticleRandom()), B=@n2)
  n4 = ScalarParameter() {defaultValue=120, parameterName=Motion Blur FPS}
  n5 = PreviousFrameSwitch(Current Frame=@n3, Previous Frame=Subtract(A=@n3, B=Divide(B=@n4)))
  n6 = ComponentMask(None=TextureCoordinate()) {g=true}
  n7 = ScalarParameter() {defaultValue=0.05, parameterName=Wing Offset, sliderMax=1}
  n8 = Subtract(A=@n5, B=Multiply(A=@n2, B=Multiply(A=@n6, B=@n7)))
  n9 = Add(
   A=Truncate!(None=@n8) {is=class MaterialExpressionTruncate}
   B=MaterialFunctionCall!(
    Gradient=Frac(None=@n8)
    Value Source=ScalarParameter() {defaultValue=0.6, parameterName=Flap Gradient Centerpoint, sliderMax=1}
    Value Destination
   ) {materialFunction=/Game/Characters/Player/Male/Medium/Bodies/M_MED_Moss_CarbideBlack/FX/Functions/MF_GradientRemapValue}.Result
  )
  n10 = MaterialFunctionCall!(
   In=ConstantBiasScale(None=Sine(None=@n9))
   Power=ScalarParameter() {defaultValue=1.5, parameterName=Flap Acceleration, sliderMax=9, sliderMin=1}
  ) {materialFunction=/Game/Characters/Player/Male/Medium/Bodies/M_MED_Moss_CarbideBlack/FX/Functions/MF_SCurve_Float}
  n11 = Multiply(
   A=Multiply(
    A=ConstantBiasScale(None=@n10.Result) {bias=-0.5, scale=2}
    B=ScalarParameter() {defaultValue=0.17, parameterName=Flapping Range, sliderMax=1}
   )
   B=ComponentMask(None=TextureCoordinate()) {desc=Flapping Mask, r=true}
  )
  n12 = MaterialFunctionCall() {materialFunction=&ObjectPivotPoint}
  n13 = WorldPosition()
  n14 = Subtract(A=@n5, B=ScalarParameter() {defaultValue=0.25, parameterName=Flutter Delay, sliderMax=1})
  n15 = Sine(None=@n14)
  n16 = ConstantBiasScale(None=Power(Base=ConstantBiasScale(None=@n15), Exp=ScalarParameter() {defaultValue=0.75, parameterName=Flutter Bias, sliderMax=1})) {bias=-0.5, scale=2}
  n17 = MaterialFunctionCall(Vector 3=Transform(None=Constant3Vector() {constant=<rgba 0 1 0 0>}) {transformSourceType=TRANSFORMSOURCE_ParticleWorld}) {materialFunction=/Engine/Functions/Engine_MaterialFunctions02/Utility/VectorLength}
  n18 = ScalarParameter() {defaultValue=7, parameterName=Flutter Range}
  n19 = MaterialFunctionCall!(Seed=@n14) {materialFunction=/Game/Characters/Player/Male/Medium/Bodies/M_MED_Moss_CarbideBlack/FX/Functions/MF_RandomValue}
  n20 = Multiply(A=Constant3Vector() {constant=<rgba 0 0 1 0>}, B=Multiply(A=Multiply(A=@n17.V3 Length, B=@n18), B=SquareRoot(None=@n19.R)))
  MaterialFunctionCall!(InVector=Add(
    A=RotateAboutAxis(NormalizedRotationAxis=@n1, RotationAngle=@n11, PivotPoint=@n12.Mesh Particle Pivot Location, Position=@n13.XYZ)
    B=Multiply(A=@n16, B=@n20)
   )) {materialFunction=/Game/Athena/Environments/Landscape/Foliage/MF_MakePreciseVec3}
  MaterialFunctionCall() {materialFunction=&DitherTemporalAA_Fortnite}
  n21 = TextureSample() {samplerType=SAMPLERTYPE_LinearColor, texture=&/T_Moss_Butterfly_BC}
  n22 = LinearInterpolate(
   A=MaterialFunctionCall(
    1=VectorParameter() {parameterName=Wing Color A, defaultValue=<rgba 1 0.31 0.1 0>}.RGB
    2=VectorParameter() {parameterName=Wing Color B, defaultValue=<rgba 1 0.955 0.9 0>}.RGB
    3=VectorParameter() {parameterName=Wing Color C, defaultValue=<rgba 0.58 1 0.1 0>}.RGB
    A=ParticleRandom()
   ) {materialFunction=&Lerp_Multiple_Float3}.Lerp 3 Inputs
   B=VectorParameter() {parameterName=Spots Color, defaultValue=<rgba 0.02 0.01 0 0>}.RGB
   Alpha=Round(None=@n21.G)
  )
  n23 = VectorParameter() {parameterName=Body Color, defaultValue=<rgba 0.005 0.005 0.005 0>}
  n24 = Round(None=@n21.R)
  n25 = TwoSidedSign()
  n26 = Saturate(None=Add(A=@n25) {constB=1.5})
  n27 = TextureCoordinate()
  n28 = Subtract(A=Saturate(None=@n25), B=@n10.Result)
  n29 = LinearInterpolate(A=Constant() {r=1}, B=Abs(None=ComponentMask(None=@n27) {desc=Flapping Mask, r=true}), Alpha=Saturate(None=Abs(None=@n28)))
  MaterialFunctionCall(Emissive=Multiply(A=LinearInterpolate(A=@n22, B=@n23.RGB, Alpha=@n24), B=Multiply(A=@n26, B=@n29))) {materialFunction=&MF_FX_EmissiveAdjustments}
  n30 = Constant() {r=220}
  n31 = Constant() {r=80}
  Multiply(
   A=Round(None=@n21.B)
   B=Saturate(None=MaterialFunctionCall(Fade Length=@n30, Fade Offset=@n31) {materialFunction=&CameraDepthFade}.Result)
  )
 ]
}
